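# Masked subsurface-profile eye material for large male player character heads, built for skeletal meshes with morph targets and mesh deformers.
# Soft and hard UV-centred sphere masks share a Metal Radius, blending sclera, external and internal iris roughness and scaling metalness.
# The same masks shade a brown eye texture, while a female Commando eye texture is scaled by an Emissive parameter.
# The shared CharacterResInEffect function supplies the character res-in effect.
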Material /Game/Characters/Player/Male/Large/Heads/Materials/Eye
graph {
 properties {
  bUsedWithMeshDeformer=true
  bUsedWithMorphTargets=true
  bUsedWithSkeletalMesh=true
  blendMode=BLEND_Masked
  pixelDepthOffsetMode=PDOM_Legacy
  shadingModel=MSM_SubsurfaceProfile
  subsurfaceProfile=/Game/Packages/Fortress_SharedMaterials/SubSurfaceProfile/EyeInner_Profile
  usedShadingModels=Subsurface Profile
 }
 nodes [
  Constant() {r=0.1333}
  n1 = TextureCoordinate()
  n2 = Constant() {r=0.5}
  n3 = ScalarParameter() {defaultValue=0.25, parameterName=Metal Radius}
  n4 = SphereMask(A=@n1, B=@n2, Radius=@n3) {hardnessPercent=0}
  n5 = SphereMask(A=@n1, B=@n2, Radius=@n3)
  LinearInterpolate(
   A=ScalarParameter() {defaultValue=0.3, parameterName=Sclera Roughness}
   B=LinearInterpolate(
    A=ScalarParameter() {defaultValue=0.1468, parameterName=External Iris Roughness}
    B=ScalarParameter() {defaultValue=1.0092, parameterName=Internal Iris Roughness}
    Alpha=@n4
   )
   Alpha=@n5
  )
  MaterialFunctionCall!(Animation Gradient, Rez In Percentage) {materialFunction=/Game/Packages/Fortress_SharedMaterials/Base_Material_Functions/CharacterResInEffect}
  Multiply(A=@n5, B=ScalarParameter() {defaultValue=1, parameterName=Metalness})
  Multiply(
   A=TextureSample() {texture=/Game/Characters/Player/Female/Medium/Heads/F_Med_Head_01/Textures/F_Med_Commando_Eye_01}.RGB
   B=ScalarParameter() {defaultValue=0.15, parameterName=Emissive}
  )
  n6 = Multiply(A=@n4) {constB=8}
  Multiply(A=Add(A=Clamp(None=@n6), B=OneMinus(None=@n5)), B=TextureSample() {texture=/Game/Characters/Player/Common/Eyes/Textures/Eye_Brown}.RGB)
 ]
}
